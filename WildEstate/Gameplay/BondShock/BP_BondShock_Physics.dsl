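; BondShock physics prop for a detached statue head that tumbles under simulation and then converts into an item pickup.
; WakeAndAddImpulse starts physics and binds hit events, which play collision audio, launch struck pawns and fire a first-impact GameplayCue, ignoring the statue.
; On sleep (authority, if enabled) or water entry, it spawns the pickup, optionally timer-delayed except in water to avoid sinking.
; The prop then SilentDies, with a GameplayCue covering the swap.
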(event EventBeginPlay
  (|CallParentFunction)
  (Networking|CanExecuteCosmeticEvents
    (:True
      (Rendering|SetVisibility (Variables|Default|GetVFXTrail) (Variables|Default|GetbIsDetachedFromPlayer) false))))

(event OnEnterWaterEvent(FortWaterInteraction)
  (params (WaterBody :type Fort%20Water%20Body%20Actor%20Object%20Reference) (WaterInteractionComponent :type Fort%20Water%20Interaction%20Component%20Object%20Reference) (bIsFirstBody :type Boolean))
  (bind _hf_spawnpickupdelay (Variables|Pickup|GetHF_SpawnPickupDelay))
  (bind _returnvalue (ScalableFloat|GetValueatLevel _hf_spawnpickupdelay))
  ;[0 0 0] Pickup
  ;;[0 0.6696 0 0.767] Spawn pickup when head physics object is no longer awake
  (Variables|Pickup|SetShouldDelaySpawnPickup false) ; cannot delay pickup or it may sink!
  (Utilities|FlowControl|DoOnce false
    (:Completed
      (if (Variables|Pickup|GetShouldDelaySpawnPickup)
        (if (> _returnvalue 0)
          (bind _returnvalue_2 (Utilities|Time|SetTimerbyEvent (EventDispatchers|CreateEvent :!event Custom|SpawnPickupUp) _returnvalue false true))
          (Variables|Pickup|SetSpawnPickupTimer _returnvalue_2)
          (else
            _))
        (else
          _))
      (CallFunction|SpawnPickupUp))) ; make sure to only do this once to \nprevent spawning multiple pickups\nWill "reset" when a new physics actor is \nspawned on subsequent pickups
)

(event Custom|WakeAndAddImpulse
  (params (Impulse :type Vector))
  (bind _staticmeshcomponent (Variables|Debug|GetStaticMeshComponent))
  ;[0 0 0] Physics
  ;;[0 0.6696 0 0.767] Start physics and fx
  (bind _returnvalue (BuildingActor|PhysicsSimulationSupport|GetPhysicsObjectComponent self))
  (Utilities|IsValid _returnvalue
    (:Is%20Valid
      (PhysicsObject|WakeUp _returnvalue "AlwaysInstant" :!class /Script/FortniteGame.FortPhysicsObjectComponent)
      (PhysicsObject|AssignOnAwakeChanged _returnvalue (EventDispatchers|CreateEvent :!event Custom|OnHeadAwakeStateChanged))
      (Physics|SetSimulatePhysics _staticmeshcomponent true)
      (Physics|AddImpulse _staticmeshcomponent Impulse "None" true)
      (CallFunction|PlayCollisionAudio Impulse)
      (Collision|AssignOnComponentHit _staticmeshcomponent (EventDispatchers|CreateEvent :!event Custom|OnHeadHit)))))

(event Custom|OnHeadAwakeStateChanged
  (params (SimulatingComponent :type Primitive%20Component%20Object%20Reference) (bIsAwake :type Boolean))
  (bind _hf_spawnpickupdelay (Variables|Pickup|GetHF_SpawnPickupDelay))
  (bind _returnvalue (ScalableFloat|GetValueatLevel _hf_spawnpickupdelay))
  ;[0 0 0] Pickup
  ;;[0 0.6696 0 0.767] Spawn pickup when head physics object is no longer awake
  (Networking|SwitchHasAuthority
    (:Authority
      (if (ScalableFloat|GetValueasBool (Variables|Default|GetRow_PickupSpawnEnabled))
        (if (not bIsAwake)
          (Utilities|FlowControl|DoOnce false
            (:Completed
              (if (Variables|Pickup|GetShouldDelaySpawnPickup)
                (if (> _returnvalue 0)
                  (bind _returnvalue_2 (Utilities|Time|SetTimerbyEvent (EventDispatchers|CreateEvent :!event Custom|SpawnPickupUp) _returnvalue false true))
                  (Variables|Pickup|SetSpawnPickupTimer _returnvalue_2)
                  (else
                    _))
                (else
                  _))
              (CallFunction|SpawnPickupUp))) ; make sure to only do this once to \nprevent spawning multiple pickups\nWill "reset" when a new physics actor is \nspawned on subsequent pickups
        )
        (else
          (Fortnite|SetCanbeDamaged self true))))))

(event Custom|OnHeadHit
  (params (HitComponent :type Primitive%20Component%20Object%20Reference) (OtherActor :type Actor%20Object%20Reference) (OtherComp :type Primitive%20Component%20Object%20Reference) (NormalImpulse :type Vector) (Hit :type Hit%20Result%20Structure%20%28by%20ref%29))
  ;[0 0 0] Physics
  ;;[1 0.8915 0.02379] OnHeadHit
  (CallFunction|ProcessHitActor OtherActor NormalImpulse)
  (if (not (Variables|Default|GetbHasFirstBigHit))
    (bind _returnvalue (GameplayTags|HasMatchingGameplayTag(Message) OtherActor (Variables|Default|GetStatueTagToCheck))) ; we don't want to play the big impact sound if the head hits the statue, \nonly when it first hits another actor
    (if (not _returnvalue)
      (CallFunction|OnFirstBigHit Hit))))

(event Custom|SpawnPickupUp
  ;[0 0 0] Pickup
  ;;[1 1 0 0.5] Spawn Pickup
  (Networking|SwitchHasAuthority
    (:Authority
      (Utilities|FlowControl|Sequence
        (:then_0
          (bind _returnvalue (Inventory|SpawnPickupInWorld (Variables|Default|GetItemToSpawn) 1 (Transformation|GetWorldLocation (Variables|Debug|GetStaticMeshComponent)) "0.000000,0.000000,1.000000" 0 false false false -1 "Other"))
          (Transformation|SetActorRotation _returnvalue (Transformation|GetActorRotation) false))
        (:then_1
          ;;;[0 0 0 0.5] Immediately destroy this physics object when the pickup is spawned. Use Silent Die to suppress death effects for a more seamless transition to the pickup object.
          (Gameplay|SilentDie false))))
    (:Remote
      _))
  (GameplayCue|ExecuteGameplayCueOnActor(Burst) self (Variables|Default|GetGCN_ConvertToPickup) (Ability|GameplayCue|MakeGameplayCueParameters :NormalizedMagnitude 0.0 :RawMagnitude 0.0 :Location (Transformation|GetActorLocation) :Normal "0,0,0" :GameplayEffectLevel 1 :AbilityLevel 1 :bReplicateLocationWhenUsingMinimalRepProxy false)))

(event Custom|OnFirstBigHit
  (params (HitResult :type Hit%20Result%20Structure))
  (bind _self self)
  ;[0 0 0] Physics
  ;;[0.6062 0 1 0.5 16] OnFirstBigHit
  ;; Multicast since this building prop doesn't have an ability system, so the GCN won't replicate automatically
  (bind _outcueparameters (GameplayCue|BuildCueParametersfromHitResult _self HitResult))
  (GameplayCue|ExecuteGameplayCueOnActor(Burst) _self (Variables|Default|GetGCN_FirstBigHit) _outcueparameters)
  (Variables|Default|SetbHasFirstBigHit true))

(event EventEndPlay
  (params (EndPlayReason :type EEndPlayReason%20Enum :default "Destroyed"))
  ;[1 0.3 0 0.5] EndPlay - Clean up timers
  (|CallParentFunction EndPlayReason)
  (Utilities|Time|ClearandInvalidateTimerbyHandle (Variables|Pickup|GetSpawnPickupTimer)))

(fn BlueprintCanInteract
  (params (InteractingPawn :type Fort%20Pawn%20Object%20Reference) (InteractionBeingAttempted :type EInteractionBeingAttempted%20Enum :default "FirstInteraction") (InteractionType :type TInteractionType%20Enum :default "IT_NoInteraction"))
  (return true))

(fn BlueprintGetInteractionString
  (params (InteractingPawn :type Fort%20Pawn%20Object%20Reference) (InteractionBeingAttempted :type EInteractionBeingAttempted%20Enum :default "FirstInteraction"))
  (return (Variables|Default|GetInteractText)))

(fn OnRep_StatueHeadLocation)

(fn PlayCollisionAudio
  (params (HitNormalImpulse :type Vector%20%28by%20ref%29))
  (Audio|OnCollision (Variables|Default|GetFortCollisionAudio) (Transformation|GetWorldLocation (Variables|Debug|GetStaticMeshComponent)) HitNormalImpulse))

(fn ProcessHitActor
  (params (HitActor :type Object%20Reference) (HitNormalImpulse :type Vector%20%28by%20ref%29))
  (Utilities|FlowControl|Sequence
    (:then_0
      (if (>= (Math|Vector|VectorLength (Variables|GetHitNormalImpulse)) (Variables|Default|GetMinMagnitudeToLaunchCharacter))
        ;[0 0 0 0.5] Launch any characters that the physics object hits
        (bind _asfort_pawn (Utilities|Casting|CastToFortPawn (Variables|GetHitActor)))
        (Character|LaunchCharacter _asfort_pawn (* (Math|Vector|Normalize (- (Transformation|GetActorLocation _asfort_pawn) (Transformation|GetActorLocation))) (Variables|Default|GetLaunchCharacterMagnitude)) false false)))
    (:then_1
      (CallFunction|PlayCollisionAudio (Variables|GetHitNormalImpulse)))))

(fn SetUseSpawnPickupDelay
  (params (Delay :type Boolean))
  (Variables|Pickup|SetShouldDelaySpawnPickup Delay))

(fn ConstructionScript)
