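; BondShock block collision: a health-based building actor that absorbs damage as the owning player's block, on that player's team.
; Damage triggers a blocked-damage event and a once-per-frame deflect cue carrying health-derived poise.
; A short-lived deflect effect uses forced net updates to drive the pawn's animation promptly.
; Break tags destroy it instantly; zero health or the owner dying or going DBNO silently kills it.
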(event EventBeginPlay
  ;[0 0.6696 0 0.767] Begin Play
  (|CallParentFunction)
  (Gameplay|ForceBuildingHealth (ScalableFloat|GetValueatLevel (Variables|Floats|GetStartingHealth)))
  (bind _asfort_player_pawn (Utilities|Casting|CastToFortPlayerPawn (Actor|GetOwner)))
  (bind _output_get (Variables|Default|SetPlayerPawn _asfort_player_pawn))
  (CallFunction|RemovalChecks _output_get)
  (CallFunction|SetTeam (Team|GetActorTeam _output_get))
  (Health|AssignOnBuildingHealthChanged (EventDispatchers|CreateEvent :!event Custom|OnBuildingHealthChanged_Event)))

(event Custom|OnBuildingHealthChanged_Event
  ;[0.3029 0.3432 1] OnHealthChanged -> Update Poise
  (Variables|Floats|SetCurrentPoise (Math|Float|NormalizetoRange (Health|GetHealth) 0.0 (Health|GetMaxHealth))))

(event Custom|RemovalChecks
  (params (PlayerPawn :type Fort%20Player%20Pawn%20Object%20Reference))
  ;[1 1 1] RemovalDelegates
  (EventDispatchers|BindEventtoOnDied PlayerPawn (delegate Custom|CustomEvent_0))
  (EventDispatchers|AssignOnPlayerStartDBNO PlayerPawn (delegate Custom|OnPlayerStartDBNO_Event_0)))

(event Custom|OnPlayerStartDBNO_Event_0
  ;[1 1 1] RemovalDelegates
  (Development|PrintString "Destroy Block Collision" false false)
  (Variables|Bools|SetOwnerDied true)
  (Gameplay|SilentDie false))

(event Custom|CustomEvent_0
  (params (DamagedActor :type Actor%20Object%20Reference) (Damage :type Float%20%28single-precision%29) (InstigatedBy :type Controller%20Object%20Reference) (DamageCauser :type Actor%20Object%20Reference) (HitLocation :type Vector) (FHitComponent :type Primitive%20Component%20Object%20Reference) (BoneName :type Name) (Momentum :type Vector))
  ;[1 1 1] RemovalDelegates
  (Development|PrintString "Destroy Block Collision" false false)
  (Variables|Bools|SetOwnerDied true)
  (Gameplay|SilentDie false))

(event Custom|StartHealingTimer)

(event Custom|ApplyHeal
  ;[1 1 1] Stop healing if we're damaged.  Wait a little while, and start healing again.
  (GameplayEffects|ApplyGameplayEffectToSelf (Ability|GetFortAbilitySystemComponent self) (Variables|Default|GetGE_Heal) 0.0 (Utilities|Struct|MakeGameplayEffectContextHandle)))

(event Custom|RemoveDeflectReaction
  (bind _playerpawn (Variables|Default|GetPlayerPawn))
  ;[1 1 1] Remove the Deflect Reaction GC and force it through for the anim blueprint (so the deflect anim stops
  (GameplayEffects|RemoveActiveGameplayEffect (Ability|GetAbilitySystemComponent _playerpawn) (Variables|Default|GetDeflectReactionHandle) -1)
  (Ability|RemoveGameplayCueReplicated (Variables|Tags|GetT_DeflectReaction))
  (Networking|ForceNetUpdate _playerpawn) ; Doing this because we're using a GC tag to run the anim.\nThis forces the GC through quickly for the other clients to see.
)

(event EventOnDamageServer
  (params (Damage :type Float%20%28single-precision%29) (DamageTags :type Gameplay%20Tag%20Container%20Structure%20%28by%20ref%29) (Momentum :type Vector :default "0, 0, 0") (HitInfo :type Hit%20Result%20Structure%20%28by%20ref%29) (InstigatedBy :type Controller%20Object%20Reference) (DamageCauser :type Actor%20Object%20Reference) (EffectContext :type Gameplay%20Effect%20Context%20Handle%20Structure))
  ;[1 1 1] OnDamageServer - Notify ability. If we receive any damage that has not been filtered out, destroy this actor
  (Utilities|FlowControl|Sequence
    (:then_0
      (if (GameplayTags|HasAnyTags DamageTags (Variables|Default|GetTagsToBreakBlock) true)
        (Variables|Bools|SetBlockBreak true) ; Certain damage tags can instantly break the block, regardless of our health
      ))
    (:then_1
      (|CallParentFunction Damage DamageTags Momentum HitInfo InstigatedBy DamageCauser EffectContext)
      (CallFunction|SendBlockedDamageEvent Damage DamageCauser)
      (Utilities|FlowControl|Sequence
        (:then_0
          (if (Variables|Bools|GetBlockBreak)
            (Actor|DestroyActor)
            (else
              (CallFunction|StartHealingTimer))))
        (:then_1
          (Utilities|DoOncePerFrame
            (:%20%20
              (CallFunction|ExecuteBlockGameplayCue Damage DamageCauser EffectContext InstigatedBy HitInfo))) ; Limit Block GC's to one per frame
        )
        (:then_2
          ;;[0.03135 0.01734 0.15] QUEST - Detach Head
          ;; Using a custom event for this even though a standard destroy might work to be consistent with the statue setup
          (Ability|SendGameplayEventtoActor InstigatedBy (Variables|Quest|GetQuest_HeadDetachedEventTag) (Utilities|Struct|MakeGameplayEventData (Variables|Quest|GetQuest_HeadDetachedEventTag) InstigatedBy self DamageCauser 0 "()" (GameplayTags|GetOwnedGameplayTags (Ability|GetFortAbilitySystemComponent InstigatedBy) :!class /Script/GameplayTags.BlueprintGameplayTagLibrary))))
        (:then_3
          (if (<= (Health|GetHealth) 0)
            (Gameplay|SilentDie false)))))))

(event GameplayCue|GameplayCue.Damage
  (params (EventType :type EGameplayCueEvent%20Enum :default "OnActive") (Parameters :type Gameplay%20Cue%20Parameters%20Structure%20%28by%20ref%29))
  (bind (_normalizedmagnitude _rawmagnitude _effectcontext _matchedtagname _originaltag _aggregatedsourcetags _aggregatedtargettags _location _normal _instigator _effectcauser) (Ability|GameplayCue|BreakGameplayCueParameters Parameters))
  ;[1 1 1] Proto Impact FX logic
  (switch Utilities|FlowControl|Switch|SwitchonEFortTeamAffiliation (Team|GetTeamAffiliation (Actor|GetOwner) _effectcauser)))

(event Custom|SpawnReflectProj
  (params (HitInfo :type Hit%20Result%20Structure) (WasGalileoBun :type Boolean))
  (bind (_bblockinghit _binitialoverlap _time _distance _location _impactpoint _normal _impactnormal _physmat _hitactor _hitcomponent) (Collision|BreakHitResult HitInfo))
  ;[1 1 1] Reflect ranged weapon hits
  (Utilities|IsValid (Variables|Default|GetReflectTarget)
    (:Is%20Valid
      (Variables|Bools|SetHitCenter (not (or (Utilities|Equal(Object) _hitcomponent 0) (Utilities|Equal(Object) _hitcomponent 0)))))))

(event EventOnDeathServer
  (params (Damage :type Float%20%28single-precision%29) (DamageTags :type Gameplay%20Tag%20Container%20Structure%20%28by%20ref%29) (Momentum :type Vector :default "0, 0, 0") (HitInfo :type Hit%20Result%20Structure%20%28by%20ref%29) (InstigatedBy :type Controller%20Object%20Reference) (DamageCauser :type Actor%20Object%20Reference) (EffectContext :type Gameplay%20Effect%20Context%20Handle%20Structure))
  ;[1 0.5 0] Activate Block Break ability if this actor was destroyed via direct damage.
  (if (Variables|Bools|GetOwnerDied)
    _))

(event EventEndPlay
  (params (EndPlayReason :type EEndPlayReason%20Enum :default "Destroyed"))
  (|CallParentFunction EndPlayReason))

(fn BlockedLobsterImpulse
  (params (HitInfo :type Hit%20Result%20Structure) (Instigator :type Character%20Object%20Reference))
  (bind (_bblockinghit _binitialoverlap _time _distance _location _impactpoint _normal _impactnormal _physmat _hitactor _hitcomponent _hitbonename _bonename _hititem _elementindex _faceindex _tracestart _traceend) (Collision|BreakHitResult HitInfo))
  (Utilities|IsValid Instigator
    (:Is%20Valid
      _)
    (:Is%20Not%20Valid
      _))
  (Variables|Floats|SetImpulseMultiplier 1.0)
  (Pawn|Character|LaunchCharacterJump (Variables|Default|GetPlayerPawn) (Math|Vector|vector+vector (* (* (Math|Vector|Normalize (- _traceend _tracestart)) (ScalableFloat|GetValueatLevel (Variables|Floats|GetImpulseXY))) (Variables|Floats|GetImpulseMultiplier)) (Math|Vector|MakeVector 0.0 0.0 (ScalableFloat|GetValueatLevel (Variables|Floats|GetImpulseZ)))) false true false true))

(fn BlueprintOnLocalInteract
  (params (InteractingPawn :type Fort%20Player%20Pawn%20Object%20Reference) (InteractionBeingAttempted :type EInteractionBeingAttempted%20Enum :default "FirstInteraction"))
  (if (Pawn|IsLocallyViewed InteractingPawn)
    _))

(fn ExecuteBlockGameplayCue
  (params (Damage :type Float%20%28double-precision%29) (DamageCauser :type Actor%20Object%20Reference) (EffectContext :type Gameplay%20Effect%20Context%20Handle%20Structure) (Instigator :type Actor%20Object%20Reference) (HitResult :type Hit%20Result%20Structure) (Color%20Hot :type Vector) (Color%20Cool :type Vector))
  (bind _playerpawn (Variables|Default|GetPlayerPawn))
  (switch Utilities|FlowControl|Switch|SwitchonEFortTeamAffiliation (Team|GetTeamAffiliation _playerpawn DamageCauser)
    (:Neutral
      _)
    (:Hostile
      _))
  ;[1 1 1] :-(  Force the GameplayCues through so the deflect anims play correctly in the animBP
  (Utilities|Time|ClearandInvalidateTimerbyHandle (Variables|Default|GetDeflectReactionTimerHandle))
  (bind _returnvalue (GameplayEffects|ApplyGameplayEffectToSelf (Ability|GetAbilitySystemComponent _playerpawn) (Variables|Default|GetGE_Deflect) 0.0 EffectContext))
  (Variables|Default|SetDeflectReactionHandle _returnvalue)
  (Networking|ForceNetUpdate _playerpawn) ; Doing this because we're using a GC tag to run the anim.\nThis forces the GC through quickly for the other clients to see.
  (bind _returnvalue_2 (Utilities|Time|SetTimerbyEvent (EventDispatchers|CreateEvent :!event RemoveDeflectReaction) 0.08))
  (Variables|Default|SetDeflectReactionTimerHandle _returnvalue_2)
  (Ability|ExecuteGameplayCueWithParamsReplicated (Variables|GameplayCues|GetGC_DeflectGeneric) (Ability|GameplayCue|MakeGameplayCueParameters :NormalizedMagnitude (Variables|Floats|GetCurrentPoise) :RawMagnitude Damage :EffectContext EffectContext :Location Color%20Hot :Normal Color%20Cool :Instigator Instigator :EffectCauser DamageCauser :GameplayEffectLevel 1 :AbilityLevel 1 :bReplicateLocationWhenUsingMinimalRepProxy false))
  (Development|PrintString "Blocked Damage! Generic" false false "(R=1.000000,G=0.694063,B=0.060146,A=1.000000)" 8.0)
  (return false))

(fn OnRep_CollisionEnabled
  (if (Math|Conversions|ToBoolean(Integer) (Variables|Default|GetCollisionEnabled))
    (Collision|SetActorEnableCollision true)
    (Development|PrintString "LOBSTER BLOCK COLLISION ENABLED" false false "(R=0.165732,G=1.000000,B=0.000000,A=1.000000)")
    (else
      (Collision|SetActorEnableCollision false)
      (Development|PrintString "LOBSTER BLOCK COLLISION DISABLED" false false "(R=1.000000,G=0.000000,B=0.038396,A=1.000000)"))))

(fn SendBlockedDamageEvent
  (params (Damage :type Float%20%28double-precision%29) (DamageCauser :type Actor%20Object%20Reference))
  (bind _playerpawn (Variables|Default|GetPlayerPawn))
  ;[1 1 1] For Challenges tracking damage blocked by the Lobster
  (Utilities|IsValid _playerpawn
    (:Is%20Valid
      (Ability|SendGameplayEventtoActor _playerpawn (Variables|Tags|GetT_BlockedDamageEvent) (Utilities|Struct|MakeGameplayEventData (Variables|Tags|GetT_BlockedDamageEvent) _playerpawn DamageCauser 0 0 "()" "(GameplayTags=)" "(GameplayTags=)" Damage)))))

(fn ConstructionScript)
