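; Gameplay ability for the BondShock helmet that gives the player a protective head collision blocker and pops the helmet off when it is destroyed.
; Activation reuses or spawns the blocker on the non-cosmetic side, attached to the animated head bone, then plays equip audio, local music and an alert.
; On blocker death, it removes the helmet item, optionally guards against duplicates, launches a physics pickup and fires a detach cue.
; Ending disables blocker collision.
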(event Ability|EventActivateAbility
  (bind _returnvalue (GameInstanceSubsystems|GetMusicEventSubsystem))
  (bind _blockingcollisionactor (Variables|Default|GetBlockingCollisionActor))
  (|CallParentFunction)
  (bind _asfort_player_pawn (Utilities|Casting|CastToFortPlayerPawn (Ability|GetAvatarActorfromActorInfo)))
  (bind _output_get (Variables|Default|SetPlayerPawn _asfort_player_pawn))
  (Utilities|FlowControl|Sequence
    (:then_0
      (CallFunction|EnableBlockingCollision)
      (Utilities|IsValid _blockingcollisionactor
        (:Is%20Valid
          (EventDispatchers|BindEventtoOnDied _blockingcollisionactor (EventDispatchers|CreateEvent :!event Custom|ProtectionDestroyed)))))
    (:then_1
      (Audio|SpawnSoundAttached (Variables|Default|GetEquipSound) (Variables|/Script/Engine.Character|GetMesh _output_get))
      ;[0.04732 0.15 0.1176 0.5] Music Stinger
      (if (Pawn|IsLocallyViewed _output_get)
        (bind _returnvalue_2 (Audio|AddEvent _returnvalue (Variables|Default|GetMusicStinger) "(TagName=\"MusicEventSubsystem.Event.Gameplay.Ability.1P\")" "(GameplayTags=)"))
        (Variables|Default|SetMusicRef _returnvalue_2)
        (Utilities|FlowControl|Delay 9.0)
        (bind _musicref (Variables|Default|GetMusicRef)
          (:then
            (Audio|RemoveEvent _returnvalue _musicref))
          (:else))))
    (:then_2
      (GameplayEventRouter|BroadcastStatefulEvent (GameplayEventRouter|GetGlobalEventRouter) (Variables|Default|GetGameplayAlertEvent) 0))))

(event Custom|ProtectionDestroyed
  (params (DamagedActor :type Actor%20Object%20Reference) (Damage :type Float%20%28single-precision%29) (InstigatedBy :type Controller%20Object%20Reference) (DamageCauser :type Actor%20Object%20Reference) (HitLocation :type Vector) (FHitComponent :type Primitive%20Component%20Object%20Reference) (BoneName :type Name) (Momentum :type Vector))
  (bind _returnvalue (Inventory|RemoveItemFromPlayer :PlayerController (Fortnite|GetFortPlayerControllerfromActor (Variables|Default|GetPlayerPawn)) :ItemDefinition "/WildEstate/Gameplay/BondShock/AGID_Helmet_BondShock.AGID_Helmet_BondShock" :AmountToRemove 1 :bForceRemoval true :bSkipVerbRemoval false))
  ;[0 0 0 0.5] Check if the item was found and removed from the inventory to prevent duplicate items spawning.
  ; If we attempted removal at this point, but didn't remove anything, do not spawn the BondShock physics object.
  (if (ScalableFloat|GetValueasBool (Variables|Default|GetbVerifyItemRemoved))
    (if (not (> _returnvalue 0))
      (Fortnite|FortShippingLog (Variables|Default|GetPlayerPawn) "Attempting to spawn BondShock pickup, but the item was not found in player's inventory. It was likely already dropped on death."))
    (else
      _))
  (bind _returnvalue_2 (Game|SpawnBuildingActorfromClass "/WildEstate/Gameplay/BondShock/BP_BondShock_Physics.BP_BondShock_Physics_C" (Math|Transform|MakeTransform (Math|Vector|MakeVector (.x (.location (Transformation|GetActorTransform DamagedActor))) (.y (.location (Transformation|GetActorTransform DamagedActor))) (+ (.z (.location (Transformation|GetActorTransform DamagedActor))) 1.2e+02)) (.rotation (Transformation|GetActorTransform DamagedActor)) (.scale (Transformation|GetActorTransform DamagedActor))) 0 false (ScalableFloat|GetValueasBool (Variables|Default|GetHFRow_ShouldDelayPickupSpawn)) true 0 0 0))
  (Utilities|IsValid _returnvalue_2
    (:Is%20Valid
      (Class|BPBondShockPhysics|WakeandAddImpulse _returnvalue_2 (Math|Vector|MakeVector (* (Math|Random|RandomFloat) (Variables|Default|GetImpulseMagnitude)) (* (Math|Random|RandomFloat) (Variables|Default|GetImpulseMagnitude)) 1.1e+03)))
    (:Is%20Not%20Valid
      _))
  (GameplayCue|ExecuteGameplayCueOnActor(Burst) (Variables|Default|GetPlayerPawn) (Variables|Default|GetGCN_HeadDetachedPlayer) (Ability|GameplayCue|MakeGameplayCueParameters :NormalizedMagnitude 0.0 :RawMagnitude 0.0 :Location "0,0,0" :Normal "0,0,0" :Instigator InstigatedBy :EffectCauser DamageCauser :SourceObject _returnvalue_2 :GameplayEffectLevel 1 :AbilityLevel 1 :bReplicateLocationWhenUsingMinimalRepProxy false)))

(event Ability|EventOnEndAbility
  (params (bWasCancelled :type Boolean))
  (|CallParentFunction bWasCancelled)
  (bind _blockingcollisionactor (Variables|Default|GetBlockingCollisionActor)
    (:then
      (Variables|/WildEstate/Gameplay/BondShock/B_BondShock_Block_Collision.B_BondShock_Block_Collision_C|SetCollisionEnabled 0 _blockingcollisionactor))
    (:else)))

(fn EnableBlockingCollision
  (bind _playerpawn (Variables|Default|GetPlayerPawn))
  (Networking|CanExecuteCosmeticEvents
    (:False
      (Variables|Default|SetHasBlockCollision false)
      ;[1 1 1] Do we already have a Blocking Collision Actor Attached?
      (for _array_element (Actor|GetAttachedActors _playerpawn true false)
        (Utilities|IsValid _array_element
          (:Is%20Valid
            (bind _asb_bond_shock_block_collision (Utilities|Casting|CastToB_BondShock_Block_Collision _array_element))
            (Variables|Default|SetHasBlockCollision true)
            (Variables|Default|SetBlockingCollisionActor _asb_bond_shock_block_collision))))
      (if (Variables|Default|GetHasBlockCollision)
        ;[1 1 1] Turn collision on if we have a blocking collision actor
        (Variables|/WildEstate/Gameplay/BondShock/B_BondShock_Block_Collision.B_BondShock_Block_Collision_C|SetCollisionEnabled 1 (Variables|Default|GetBlockingCollisionActor))
        (else
          ;[1 1 1] Spawn Blocking Collision Actor if none were found
          (bind _returnvalue (Game|SpawnActorfromClass (Variables|Default|GetActorClass_BlockCollision) (Math|Transform|MakeTransform (Math|Vector|vector+vector (Transformation|GetActorLocation _playerpawn) (Math|Vector|MakeVector 0.0 0.0 (Variables|Default|GetBlockCollisionActorAttachZOffset)) (* (Transformation|GetActorForwardVector _playerpawn) (Variables|Default|GetBlockCollisionActorAttachXOffset))) (Transformation|GetActorRotation _playerpawn)) "AlwaysSpawn" "SelectDefaultAtRuntime" _playerpawn _playerpawn))
          (bind _asb_bond_shock_block_collision_2 (Utilities|Casting|CastToB_BondShock_Block_Collision _returnvalue))
          (bind _output_get (Variables|Default|SetBlockingCollisionActor _asb_bond_shock_block_collision_2))
          (Transformation|AttachActorToComponent _output_get (Variables|/Script/Engine.Character|GetMesh (Variables|Default|GetPlayerPawn)) "head" "KeepWorld" "KeepWorld" "KeepRelative" true) ; attaching to the skeletal mesh component will animation with the \nplayer movements, preventing the case where the head collision \nis outside the protective radius (FORT-994856)
        )))))
